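; BP_LevelsChecker is an art tool that drives a post-process material to check base color levels against min/max values, with an optional metallic minimum.
; The construction script creates a dynamic instance of MAT_LevelsChecker_BaseColor if needed, then pushes thresholds, max distance and marker colors into it.
; The Enabled flag calls enable or disable, which set the material's blendable weight on the PostProcess component to one or zero.
; Setter functions update min and max live.
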(event EventBeginPlay)

(event Collision|EventActorBeginOverlap
  (params (OtherActor :type Actor%20Object%20Reference)))

(event EventTick
  (params (DeltaSeconds :type Float%20%28single-precision%29)))

(fn disable
  (Utilities|Struct|SetmembersinPostProcessSettings (Variables|/Script/Engine.PostProcessComponent|GetSettings (Variables|Default|GetPostProcess)) (Utilities|Struct|MakeWeightedBlendables (Utilities|Array|MakeArray (Utilities|Struct|MakeWeightedBlendable 0.0 (Variables|Default|GetdynamicMaterialPP))))))

(fn enable
  (Utilities|Struct|SetmembersinPostProcessSettings (Variables|/Script/Engine.PostProcessComponent|GetSettings (Variables|Default|GetPostProcess)) (Utilities|Struct|MakeWeightedBlendables (Utilities|Array|MakeArray (Utilities|Struct|MakeWeightedBlendable 1.0 (Variables|Default|GetdynamicMaterialPP))))))

(fn setMaxValue
  (params (MaxValue :type Float%20%28double-precision%29))
  (Rendering|Material|SetScalarParameterValue (Variables|Default|GetdynamicMaterialPP) "MaxValue" MaxValue :!class /Script/Engine.MaterialInstanceDynamic)
  (Variables|Default|SetmaxValue MaxValue))

(fn setMinValue
  (params (minValue :type Float%20%28double-precision%29))
  (Rendering|Material|SetScalarParameterValue (Variables|Default|GetdynamicMaterialPP) "MinValue" minValue :!class /Script/Engine.MaterialInstanceDynamic)
  (Variables|Default|SetminValue minValue))

(fn ConstructionScript
  (if (not (Utilities|IsValid (Variables|Default|GetdynamicMaterialPP) :!class /Script/Engine.KismetSystemLibrary))
    (bind _returnvalue (Rendering|Material|CreateDynamicMaterialInstance "/Game/ArtTools/LevelsChecker/MAT_LevelsChecker_BaseColor.MAT_LevelsChecker_BaseColor" "None" "None" :!class /Script/Engine.KismetMaterialLibrary))
    (Variables|Default|SetdynamicMaterialPP _returnvalue))
  (Rendering|Material|SetScalarParameterValue (Variables|Default|GetdynamicMaterialPP) "MinValue" (Variables|Default|GetminValue) :!class /Script/Engine.MaterialInstanceDynamic)
  (Rendering|Material|SetScalarParameterValue (Variables|Default|GetdynamicMaterialPP) "MinMetallicValue" (* (Math|Conversions|ToInteger(Boolean) (Variables|Default|GetcheckMetallic)) (Variables|Default|GetminMetallicValue)) :!class /Script/Engine.MaterialInstanceDynamic)
  (Rendering|Material|SetScalarParameterValue (Variables|Default|GetdynamicMaterialPP) "MaxValue" (Variables|Default|GetmaxValue) :!class /Script/Engine.MaterialInstanceDynamic)
  (Rendering|Material|SetScalarParameterValue (Variables|Default|GetdynamicMaterialPP) "MaxDistance" (Variables|Default|GetmaxDistance) :!class /Script/Engine.MaterialInstanceDynamic)
  (Rendering|Material|SetVectorParameterValue (Variables|Default|GetdynamicMaterialPP) "MinValueColor" (Variables|Default|GetminValueColor) :!class /Script/Engine.MaterialInstanceDynamic)
  (Rendering|Material|SetVectorParameterValue (Variables|Default|GetdynamicMaterialPP) "MinMetallicValueColor" (Variables|Default|GetminMetallicValueColor) :!class /Script/Engine.MaterialInstanceDynamic)
  (Rendering|Material|SetVectorParameterValue (Variables|Default|GetdynamicMaterialPP) "MaxValueColor" (Variables|Default|GetmaxValueColor) :!class /Script/Engine.MaterialInstanceDynamic)
  (if (Variables|Default|GetEnabled)
    (CallFunction|Enable)
    (else
      (CallFunction|Disable))))
