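# Post-process debug material that paints warning colors over the scene where stored base color is out of range.
# If nodes compare base color values against minimum, maximum and metallic-adjusted minimum thresholds, the last raised by a power curve on the Metallic buffer.
# Values derive from desaturated luminance and a channel max that reads red and blue but skips green.
# A scene-depth SphereMask limits the effect by MaxDistance; ES2/ES3_1 get flat red.
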Material /Game/ArtTools/LevelsChecker/MAT_LevelsChecker_BaseColor
graph {
 properties {materialDomain=MD_PostProcess, pixelDepthOffsetMode=PDOM_Legacy, usedShadingModels=Default Lit}
 nodes [
  n1 = ComponentMask(None=SceneTexture() {sceneTextureId=PPI_SceneDepth}.Color) {r=true}
  n2 = Constant()
  n3 = ScalarParameter() {defaultValue=3000, parameterName=MaxDistance}
  n4 = Constant() {r=100}
  n5 = Constant3Vector() {constant=<rgba 1 0 0 0>}
  FeatureLevelSwitch(
   Default=LinearInterpolate(A=NamedRerouteUsage(), B=NamedRerouteUsage(), Alpha=SphereMask(A=@n1, B=@n2, Radius=@n3, Hardness=@n4))
   ES2=@n5
   ES3_1=@n5
  )
  n6 = ComponentMask(None=SceneTexture() {sceneTextureId=PPI_Metallic}.Color) {r=true}
  NamedRerouteDeclaration(None=LinearInterpolate(
    A=ScalarParameter() {defaultValue=0.009, parameterName=MinValue}
    B=ScalarParameter() {defaultValue=0.25, parameterName=MinMetallicValue}
    Alpha=Power(Base=@n6) {constExponent=3}
   )) {name=adjustedMinValue, nodeColor=<rgba 1 0 0.3281 1>}
  NamedRerouteDeclaration(None=ComponentMask(None=SceneTexture() {sceneTextureId=PPI_PostProcessInput0}.Color) {b=true, g=true, r=true}) {name=Lit, nodeColor=<rgba 0 0.9766 1 1>}
  n7 = NamedRerouteUsage()
  NamedRerouteDeclaration(None=If(
    A=NamedRerouteUsage()
    B=ScalarParameter() {defaultValue=0.009, parameterName=MinValue}
    "A > B"=@n7
    "A == B"=@n7
    "A < B"=VectorParameter() {parameterName=MinValueColor, defaultValue=<rgba 1 1 0 1>}.RGB
   )) {name=warnings, nodeColor=<rgba 1 0 0.5391 1>}
  n8 = SceneTexture() {sceneTextureId=PPI_StoredBaseColor}
  n9 = Constant() {r=1}
  NamedRerouteDeclaration(None=MaterialFunctionCall(Float3=Desaturation(None=@n8.Color, Fraction=@n9)) {materialFunction=&BreakOutFloat3Components}.R) {name=luminance, nodeColor=<rgba 0.3359 0 1 1>}
  n10 = NamedRerouteUsage()
  NamedRerouteDeclaration(None=If(
    A=NamedRerouteUsage()
    B=ScalarParameter() {defaultValue=0.9, parameterName=MaxValue}
    "A > B"=VectorParameter() {parameterName=MaxValueColor, defaultValue=<rgba 1 0 0.8328 1>}.RGB
    "A == B"=@n10
    "A < B"=@n10
   )) {name=MaxValueCheck, nodeColor=<rgba 0.8984 1 0 1>}
  n11 = NamedRerouteUsage()
  NamedRerouteDeclaration(None=If(
    A=NamedRerouteUsage()
    B=NamedRerouteUsage()
    "A > B"=@n11
    "A == B"=@n11
    "A < B"=VectorParameter() {parameterName=MinMetallicValueColor, defaultValue=<rgba 1 0 0 1>}.RGB
   )) {name=MetallicCheck, nodeColor=<rgba 1 0.6328 0 1>}
  n12 = MaterialFunctionCall(Float3=@n8.Color) {materialFunction=&BreakOutFloat3Components}
  NamedRerouteDeclaration(None=Max(A=Max(A=@n12.R, B=@n12.R), B=@n12.B)) {name=reflectance, nodeColor=<rgba 0.9219 0 1 1>}
  NamedRerouteUsage()
  NamedRerouteUsage()
  SceneDepth()
 ]
}
