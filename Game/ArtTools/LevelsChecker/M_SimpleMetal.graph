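# M_SimpleMetal is a plain Default Lit metal material in the LevelsChecker art tools folder, enabled for use with Nanite.
# It exposes scalar parameters for BaseColor, Metallic, Roughness and Specular, defaulting to a dark, fully metallic, fairly smooth surface.
# BaseColor is a single grayscale value, and the graph has no textures or math nodes.
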Material /Game/ArtTools/LevelsChecker/M_SimpleMetal
graph {
 properties {bUsedWithNanite=true, pixelDepthOffsetMode=PDOM_Legacy, usedShadingModels=Default Lit}
 nodes [
  ScalarParameter() {defaultValue=0.25, parameterName=BaseColor}
  ScalarParameter() {defaultValue=1, parameterName=Metallic}
  ScalarParameter() {defaultValue=0.3, parameterName=Roughness}
  ScalarParameter() {parameterName=Specular}
 ]
}
